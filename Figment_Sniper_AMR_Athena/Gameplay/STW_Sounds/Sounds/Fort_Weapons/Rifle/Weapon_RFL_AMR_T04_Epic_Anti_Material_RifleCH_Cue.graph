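# Sound cue for the Epic T04 anti-material rifle that plays a single supercharge layer wave.
# The wave passes through an attenuation node using the shared small-rifle weapon preset, then a sound class node that assigns the third-person ranged weapon fire class.
# The cue overrides concurrency with a prevent-new rule, so new instances are refused once the limit is reached.
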SoundCue /Figment_Sniper_AMR_Athena/Gameplay/STW_Sounds/Sounds/Fort_Weapons/Rifle/Weapon_RFL_AMR_T04_Epic_Anti_Material_RifleCH_Cue
root=SoundClass_25
nodes [
 {id=Attenuation_35, children [WavePlayer_18], properties {attenuationSettings=/Game/Sounds/Attenuation/Fort_Weapon_Rifle_Sml_Att}}
 {
  id=SoundClass_25
  children [Attenuation_35]
  properties {soundClassOverride=&Fort_WeaponRanged_Fire_3P_PSM}
 }
 {
  id=WavePlayer_18
  properties {soundWaveAssetPtr=&/amr_supercharge_layer}
 }
]
properties {
 bOverrideConcurrency=true
 maxDistance=3000
 volumeMultiplier=1
 attenuationOverrides {distanceAlgorithm=NaturalSound, falloffDistance=4000, attenuationShapeExtents {x=1000}}
 concurrencyOverrides {resolutionRule=PreventNew}
}
