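# Sound Cue named for the impact sound of the tier 4 AMR rifle, routed to the Fort_Impacts sound class.
# It has no nodes and no root, so as shown it produces no sound.
# It lowers the volume, overrides concurrency to prevent new instances at the limit, and stores natural-sound attenuation values without showing the attenuation override enabled.
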SoundCue /Figment_Sniper_AMR_Athena/Gameplay/STW_Sounds/Sounds/Fort_Weapons/Rifle/Weapon_RFL_AMR_T04_Impact_Cue
root=null
nodes []
properties {
 bOverrideConcurrency=true
 soundClassObject=&Fort_Impacts
 volumeMultiplier=0.7
 attenuationOverrides {distanceAlgorithm=NaturalSound, falloffDistance=4000, attenuationShapeExtents {x=1000}}
 concurrencyOverrides {resolutionRule=PreventNew}
}
