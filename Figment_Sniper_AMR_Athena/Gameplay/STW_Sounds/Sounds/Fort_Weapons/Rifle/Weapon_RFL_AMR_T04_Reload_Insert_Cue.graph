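# Sound cue for the magazine-insert foley of the Tier 4 AMR rifle reload.
# A single wave player plays a raygun reload-insert sample from the AMR_Electric folder, with no randomization or modulation.
# It uses the weapon foley sound class and shared reload attenuation at reduced volume.
# Its concurrency override prevents new instances once the limit is reached.
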SoundCue /Figment_Sniper_AMR_Athena/Gameplay/STW_Sounds/Sounds/Fort_Weapons/Rifle/Weapon_RFL_AMR_T04_Reload_Insert_Cue
root=WavePlayer_24
nodes [{
  id=WavePlayer_24
  properties {soundWaveAssetPtr=/Figment_Sniper_AMR_Athena/Gameplay/STW_Sounds/Sounds/Fort_Weapons/Rifle/AMR_Electric/weapon_raygun_reload_insert}
 }]
properties {
 attenuationSettings=&Fort_Weapon_Reload_Att
 bOverrideConcurrency=true
 maxDistance=2250
 soundClassObject=&Fort_WeaponFoley
 volumeMultiplier=0.702
 attenuationOverrides {distanceAlgorithm=NaturalSound, falloffDistance=4000, attenuationShapeExtents {x=1000}}
 concurrencyOverrides {resolutionRule=PreventNew}
}
