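# Sound cue for the reload-start sound of the tier 4 AMR sniper rifle.
# A single wave player plays an electric raygun reload-start wave, with boosted volume, in a third-person ranged weapon sound class.
# It uses a shared weapon reload attenuation asset, and its concurrency override blocks new instances rather than stopping playing ones.
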SoundCue /Figment_Sniper_AMR_Athena/Gameplay/STW_Sounds/Sounds/Fort_Weapons/Rifle/Weapon_RFL_AMR_T04_Reload_Start_Cue
root=WavePlayer_15
nodes [{
  id=WavePlayer_15
  properties {soundWaveAssetPtr=/Figment_Sniper_AMR_Athena/Gameplay/STW_Sounds/Sounds/Fort_Weapons/Rifle/AMR_Electric/weapon_raygun_reload_start}
 }]
properties {
 attenuationSettings=&Fort_Weapon_Reload_Att
 bOverrideConcurrency=true
 maxDistance=2250
 soundClassObject=&Fort_WeaponRanged_Fire_3P_PSM
 volumeMultiplier=1.4
 attenuationOverrides {distanceAlgorithm=NaturalSound, falloffDistance=4000, attenuationShapeExtents {x=1000}}
 concurrencyOverrides {resolutionRule=PreventNew}
}
